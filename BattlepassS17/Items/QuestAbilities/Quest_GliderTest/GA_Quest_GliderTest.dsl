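; Quest gameplay ability for a glider test that applies cosmetic item and variant overrides to the player while it is active.
; Activation and end call ToggleCosmetics with Active and Inactive states respectively.
; ToggleCosmetics swaps cosmetics on the avatar's Fort player pawn, using a GUID generated from the quest ID and the Player Cosmetic Overrides variable.
; Debug prints log each start, toggle and end.
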(event Ability|EventActivateAbility
  (CallFunction|ToggleCosmetics "Active")
  (Development|PrintString "GliderTest: Ability Started" true true "(R=0.000000,G=0.660000,B=1.000000,A=1.000000)" 6e+01))

(event Custom|ToggleCosmetics
  (params (ForcedSwapState :type EFortAppliedSwapItemAndVariantState%20Enum))
  (bind _returnvalue (GUID|GenerateGuidfromObject (Variables|Default|GetQuestID)))
  (bind _asfort_player_pawn (Utilities|Casting|CastToFortPlayerPawn (Ability|GetAvatarActorfromActorInfo self)))
  (Class|FortPlayerPawn|SwapPawnCosmeticItemandVariant _asfort_player_pawn _returnvalue (Variables|Default|GetPlayer%20Cosmetic%20Overrides) false ForcedSwapState false)
  (Development|PrintString (Utilities|String|Append "GliderTest: ToggleState = " (Utilities|String|EnumtoString ForcedSwapState)) true true "(R=0.000000,G=0.660000,B=1.000000,A=1.000000)" 6e+01))

(event Ability|EventOnEndAbility
  (params (bWasCancelled :type Boolean))
  (CallFunction|ToggleCosmetics "Inactive")
  (Development|PrintString "GliderTest: Ability Ended" true true "(R=0.000000,G=0.660000,B=1.000000,A=1.000000)" 6e+01))
